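# Niagara system for the Edge vehicle boost cosmetic, emitting GPU sprite exhaust at the car's boost sockets with painted color variants.
# BCKP_LocalSprites002 makes short-lived sprites with curl noise and orbital motion; BCKP_Sprites001 makes longer-lived trailing sprites inheriting source movement.
# Each merges several layers via spawn groups, which select size, velocity and drag and set VisTag alongside four sprite renderers.
# User parameters supply offsets, paint tuning and frontend-preview motion changes.
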NiagaraSystem /VehicleCosmetics/Boosts/Boost_Edge/FX/NS_Boost_Edge
dependencies {
 usedDataInterfaces [
  &NiagaraDataInterfaceCurve
  &NiagaraDataInterfaceSpriteRendererInfo
  &NiagaraDataInterfaceSkeletalMesh
 ]
 usedDynamicInputs []
 usedModules [
  &self:SystemSpawnScript.SystemScriptSource.SystemScriptGraph.NiagaraNodeAssignment_0.SetVariables_#1
  &SystemState
 ]
}
emitters {
 BCKP_LocalSprites002 {
  simTarget=GPUComputeSim
  emitterUpdateScript [
   &EmitterState()
   &VC_FilteredSocketCount()
   &SpawnRate(SpawnRate=60, Use Spawn Probability=false)
   &SpawnRate(SpawnGroup=1, SpawnRate=30, Use Spawn Probability=false) {module=SpawnRate001}
   &SpawnRate(SpawnGroup=2, SpawnRate=20, Use Spawn Probability=false) {module=SpawnRate002}
  ]
  particleSpawnScript [
   &VC_PaintedVariantsOverride(
    Black=<rgba 0.11 0.11 0.11 1>
    BurntSienna=<rgba 0.1 0.0588 0.035 1>
    Cobalt=<rgba 0.1375 0.175 0.25 1>
    Crimson=<rgba 0.95 0 0.0257 1>
    DefaultColor=<rgba 0.7 0.35 0.7 1>
    ForestGreen=<rgba 0 0.6 0.0303 1>
    Gold:LinearColor=<rgba 0.7 0.4521 0.105 1>
    Grey=<rgba 0.4 0.4 0.4 1>
    Lime=<rgba 0.52 0.65 0.13 1>
    Orange=<rgba 0.8 0.4267 0 1>
    Pink=<rgba 0.6 0.3 0.4909 1>
    Purple=<rgba 0.4667 0 0.8 1>
    R1 Black:LinearColor=<rgba 0 0 0 1>
    R1 BurntSienna:LinearColor=<rgba 0.0642 0.0198 0.0079 1>
    R1 Cobalt:LinearColor=<rgba 0.01 0.0205 0.1 1>
    R1 Crimson:LinearColor=<rgba 0.09 0 0.0024 1>
    R1 DefaultColor=<rgba 0.2 0.0394 0.0073 1>
    R1 ForestGreen:LinearColor=<rgba 0 0.07 0.0035 1>
    R1 Gold:LinearColor=<rgba 0.05 0.0292 0 1>
    R1 Grey:LinearColor=<rgba 0.02 0.02 0.02 1>
    R1 Lime:LinearColor=<rgba 0.075 0.15 0 1>
    R1 Orange:LinearColor=<rgba 0.14 0.0535 0.0004 1>
    R1 Pink:LinearColor=<rgba 0.18 0 0.09 1>
    R1 Purple:LinearColor=<rgba 0.025 0 0.05 1>
    R1 Saffron:LinearColor=<rgba 0.07 0.07 0 1>
    R1 SkyBlue:LinearColor=<rgba 0 0.0333 0.05 1>
    R1 TitaniumWhite:LinearColor=<rgba 0.07 0.07 0.07 1>
    R2 Black:LinearColor=<rgba 0.0675 0.0675 0.075 1>
    R2 BurntSienna:LinearColor=<rgba 0.3 0.11 0 1>
    R2 Cobalt:LinearColor=<rgba 0.075 0.2247 0.3 1>
    R2 Crimson:LinearColor=<rgba 0.2 0 0.0054 1>
    R2 DefaultColor:LinearColor=<rgba 1 0.2667 0.2 1>
    R2 ForestGreen:LinearColor=<rgba 0 0.5 0.0253 1>
    R2 Gold:LinearColor=<rgba 0.8417 1 0.1 1>
    R2 Grey:LinearColor=<rgba 0.22 0.22 0.22 1>
    R2 Orange:LinearColor=<rgba 2 0.6976 0 1>
    R2 Pink:LinearColor=<rgba 0.3 0.24 0.2782 1>
    R2 Purple:LinearColor=<rgba 0.1642 0 0.4 1>
    R3 Black:LinearColor=<rgba 0.162 0.162 0.18 1>
    R3 BurntSienna:LinearColor=<rgba 0.65 0.3824 0.2275 1>
    R3 Cobalt:LinearColor=<rgba 0.3 0.375 0.75 1>
    R3 Crimson:LinearColor=<rgba 0.25 0 0.0068 1>
    R3 DefaultColor:LinearColor=<rgba 0.4041 0.225 0.45 1>
    R3 ForestGreen:LinearColor=<rgba 0 0.28 0.0141 1>
    R3 Gold:LinearColor=<rgba 0.3 0.2125 0.09 1>
    R3 Grey:LinearColor=<rgba 0.2 0.2 0.2 1>
    R3 Lime:LinearColor=<rgba 0.28 0.35 0.21 1>
    R3 Orange:LinearColor=<rgba 0.33 0.2226 0.165 1>
    R3 Pink:LinearColor=<rgba 0.3 0.156 0.2476 1>
    R3 Purple:LinearColor=<rgba 0.1314 0.028 0.28 1>
    R3 Saffron:LinearColor=<rgba 0.6417 0.7 0 1>
    R3 SkyBlue:LinearColor=<rgba 0.21 0.5367 0.7 1>
    R3 TitaniumWhite:LinearColor=<rgba 1 1 1 1>
    R4 DefaultColor:LinearColor=$User.OutlineParticleColor
    Saffron=<rgba 0.4 0.3733 0.08 1>
    SkyBlue=<rgba 0.15 0.4614 0.5 1>
    TitaniumWhite=<rgba 0.4 0.4 0.4 1>
    Use R1 Painted Overrides=true
    Use R2 Painted Overrides=true
    Use R3 Painted Overrides=true
    Use R4 Painted Overrides=true
   )
   /SetVariables(Particles.VisTag:NiagaraInt32=&SpawnGroup()) {isSetParameters=true}
   &InitializeParticle(
    Color Mode=::NewEnumerator0
    Lifetime=/VehicleCosmetics/FX/Niagara/Modules/MergeEmitters/VC_FloatBySpawnGroup(
     SpawnGroup or ExecIndex:NiagaraInt32=0
     CustomInt:NiagaraInt32=0
     MaximumSpawnGroups:NiagaraInt32=3
     Float SG 0:NiagaraFloat=&RandomRangeFloat(Minimum=0.3, Maximum=0.7)
     Float SG 1:NiagaraFloat=&RandomRangeFloat(Minimum=0.3, Maximum=0.7)
     Float SG 2:NiagaraFloat=0.4
     Float SG 3:NiagaraFloat=0
     Float SG 4:NiagaraFloat=0
     Float SG 5:NiagaraFloat=0
    )
    Sprite Rotation Angle=&RandomRangeFloat(Minimum=-1, Maximum=1)
    Sprite Rotation Mode=::NewEnumerator1
    Sprite Size Mode=::NewEnumerator3
    Uniform Sprite Size=&VC_FloatBySpawnGroup(
     SpawnGroup or ExecIndex:NiagaraInt32=0
     CustomInt:NiagaraInt32=0
     MaximumSpawnGroups:NiagaraInt32=3
     Float SG 0:NiagaraFloat=&RandomRangeFloat(Minimum=1, Maximum=5)
     Float SG 1:NiagaraFloat=&RandomRangeFloat(Minimum=3, Maximum=9)
     Float SG 2:NiagaraFloat=100
     Float SG 3:NiagaraFloat=0
     Float SG 4:NiagaraFloat=0
     Float SG 5:NiagaraFloat=0
    )
    UsePositionOffset=false
    Write Material Random=true
   )
   &PartitionParticles(Number Of Partitions=$Emitter.PortCount)
   &SkeletalMeshLocation(
    Bone / Socket Index=$Particles.PartitionParticles.Partition
    Bone / Socket Sampling Mode=::NewEnumerator5
    MaskSpawnGroup=false
    Mesh Sampling Type=::NewEnumerator6
    Sampled Position Offset=&AddVectorToPosition(
     Position=/VehicleCosmetics/FX/Niagara/Modules/MergeEmitters/VC_VectorBySpawnGroup(
      SpawnGroup or ExecIndex:NiagaraInt32=0
      CustomInt:NiagaraInt32=0
      MaximumSpawnGroups:NiagaraInt32=3
      Vector Group 0:Vector3f=<-100 0 0>
      Vector Group 1:Vector3f=<-150 0 0>
      Vector Group 2:Vector3f=<-110 0 0>
      Vector Group 3:Vector3f=<0 0 0>
      Vector Group 4:Vector3f=<0 0 0>
      Vector Group 5:Vector3f=<0 0 0>
     )
     Vector=&AddVectorToPosition(Position=$User.SingleBoostPos, Vector=$User.DTCarOffset)
    )
    Skeletal Mesh=$System.SkeletalMesh
   )
   &AddVelocity(Velocity=<-1500 0 0>) {enabled=false}
   &AddVelocity(Velocity=&MakeCustomVectorFromBool(
     Bool=$User.bIsFrontendPreview
     True Vector=&MakeVector(
      X=&VC_FloatBySpawnGroup(
       SpawnGroup or ExecIndex:NiagaraInt32=0
       CustomInt:NiagaraInt32=0
       MaximumSpawnGroups:NiagaraInt32=4
       Float SG 0:NiagaraFloat=-800
       Float SG 1:NiagaraFloat=-600
       Float SG 2:NiagaraFloat=-300
       Float SG 3:NiagaraFloat=0
       Float SG 4:NiagaraFloat=0
       Float SG 5:NiagaraFloat=0
      )
      Y=0
      Z=0
     )
     False Vector=&MakeVector(
      X=&VC_FloatBySpawnGroup(
       SpawnGroup or ExecIndex:NiagaraInt32=0
       CustomInt:NiagaraInt32=0
       MaximumSpawnGroups:NiagaraInt32=3
       Float SG 0:NiagaraFloat=0
       Float SG 1:NiagaraFloat=0
       Float SG 2:NiagaraFloat=-300
       Float SG 3:NiagaraFloat=0
       Float SG 4:NiagaraFloat=0
       Float SG 5:NiagaraFloat=0
      )
      Y=0
      Z=0
     )
    )) {module=AddVelocity002}
   &AddVelocity(
    Velocity=&MakeVector(
     X=&VC_FloatBySpawnGroup(
      SpawnGroup or ExecIndex:NiagaraInt32=0
      CustomInt:NiagaraInt32=0
      MaximumSpawnGroups:NiagaraInt32=3
      Float SG 0:NiagaraFloat=&RandomRangeFloat(Minimum=-50, Maximum=-25)
      Float SG 1:NiagaraFloat=0
      Float SG 2:NiagaraFloat=0
      Float SG 3:NiagaraFloat=0
      Float SG 4:NiagaraFloat=0
      Float SG 5:NiagaraFloat=0
     )
     Y=&VC_FloatBySpawnGroup(
      SpawnGroup or ExecIndex:NiagaraInt32=0
      CustomInt:NiagaraInt32=0
      MaximumSpawnGroups:NiagaraInt32=3
      Float SG 0:NiagaraFloat=&RandomRangeFloat(Minimum=-150, Maximum=150)
      Float SG 1:NiagaraFloat=0
      Float SG 2:NiagaraFloat=0
      Float SG 3:NiagaraFloat=0
      Float SG 4:NiagaraFloat=0
      Float SG 5:NiagaraFloat=0
     )
     Z=&VC_FloatBySpawnGroup(
      SpawnGroup or ExecIndex:NiagaraInt32=0
      CustomInt:NiagaraInt32=0
      MaximumSpawnGroups:NiagaraInt32=3
      Float SG 0:NiagaraFloat=&RandomRangeFloat(Minimum=-300, Maximum=300)
      Float SG 1:NiagaraFloat=0
      Float SG 2:NiagaraFloat=0
      Float SG 3:NiagaraFloat=0
      Float SG 4:NiagaraFloat=0
      Float SG 5:NiagaraFloat=0
     )
    )
    Velocity Speed Scale=1.5
   ) {module=AddVelocity003}
   &ShapeLocation(
    Cylinder Height=&VC_FloatBySpawnGroup(
     SpawnGroup or ExecIndex:NiagaraInt32=0
     CustomInt:NiagaraInt32=0
     MaximumSpawnGroups:NiagaraInt32=3
     Float SG 0:NiagaraFloat=20
     Float SG 1:NiagaraFloat=25
     Float SG 2:NiagaraFloat=0
     Float SG 3:NiagaraFloat=0
     Float SG 4:NiagaraFloat=0
     Float SG 5:NiagaraFloat=0
    )
    Cylinder Radius=&VC_FloatBySpawnGroup(
     SpawnGroup or ExecIndex:NiagaraInt32=0
     CustomInt:NiagaraInt32=0
     MaximumSpawnGroups:NiagaraInt32=3
     Float SG 0:NiagaraFloat=15
     Float SG 1:NiagaraFloat=20
     Float SG 2:NiagaraFloat=0
     Float SG 3:NiagaraFloat=0
     Float SG 4:NiagaraFloat=0
     Float SG 5:NiagaraFloat=0
    )
    MaskSpawnGroup=false
    Override Seed=false
    Shape Primitive=::NewEnumerator1
    Surface Only=false
    UseOwnerScale=true
   )
   &AddRotationalVelocity(Rotation Rate=&VectorFromFloat(Value=&VC_FloatBySpawnGroup(
      SpawnGroup or ExecIndex:NiagaraInt32=0
      CustomInt:NiagaraInt32=0
      MaximumSpawnGroups:NiagaraInt32=3
      Float SG 0:NiagaraFloat=&RandomRangeFloat(Minimum=-0.75, Maximum=0.75)
      Float SG 1:NiagaraFloat=&RandomRangeFloat(Minimum=-0.75, Maximum=0.75)
      Float SG 2:NiagaraFloat=0
      Float SG 3:NiagaraFloat=0
      Float SG 4:NiagaraFloat=0
      Float SG 5:NiagaraFloat=0
     )))
   &ApplyInitialForces(
    Apply Force to Velocity:NiagaraBool=/Niagara/DynamicInputs/Helpers/MaskBoolBySpawnGroup(Masked Bool:NiagaraBool=true, Passthrough Bool:NiagaraBool=false, Spawn Group Mask:NiagaraInt32=0)
    Apply Rotational Force to Rotational Velocity:NiagaraBool=&MaskBoolBySpawnGroup(Masked Bool:NiagaraBool=true, Passthrough Bool:NiagaraBool=false, Spawn Group Mask:NiagaraInt32=0)
   )
  ]
  particleUpdateScript [
   &ParticleState()
   /SetVariables(Particles.Fire:LinearColor=&MultiplyLinearColor(
     Linear Color=$Particles.Color
     Scale Factor=&MakeVector4FromVectorAndFloat(
      "Vector (RGB)"=&Multiply_VectorByFloat(
       Vector=<1 1 1>
       Float=&MakeCustomFloatFromBool(Bool=$User.bIsPainted, True Float=$User.FireValuePainted, False Float=1)
      )
      "Float (Alpha)"=&FloatFromCurve(FloatCurve=DataInterface() {propertyValues {Curve {defaultValue=1, keys [{tangentMode=RCTM_User, time=0, value=0}, {tangentMode=RCTM_User, time=1, value=0.5}]}}})
     )
    )) {isSetParameters=true}
   &ScaleSpriteSize(
    Scale Sprite Size Mode=::NewEnumerator0
    Uniform Scale Factor=&VC_FloatBySpawnGroup(
     SpawnGroup or ExecIndex:NiagaraInt32=0
     CustomInt:NiagaraInt32=0
     MaximumSpawnGroups:NiagaraInt32=3
     Float SG 0:NiagaraFloat=&FloatFromCurve(
      FloatCurve=DataInterface() {propertyValues {Curve {keys [
          <at 0 1 RCIM_Cubic>
          {arriveTangent=-0.9025, interpMode=RCIM_Cubic, leaveTangent=-0.9025, tangentMode=RCTM_User, time=0.22, value=1.25}
          {arriveTangent=-1.7745, interpMode=RCIM_Cubic, leaveTangent=-1.7745, tangentMode=RCTM_User, time=0.4, value=1.5}
          {arriveTangent=-0.272, interpMode=RCIM_Cubic, leaveTangent=-0.272, tangentMode=RCTM_User, time=0.6, value=2}
          <at 1 0 RCIM_Cubic>
         ]}}}
      Scale Curve=2.5
     )
     Float SG 1:NiagaraFloat=&FloatFromCurve(
      FloatCurve=DataInterface() {propertyValues {Curve {keys [
          <at 0 1 RCIM_Cubic>
          {arriveTangent=-0.9025, interpMode=RCIM_Cubic, leaveTangent=-0.9025, tangentMode=RCTM_User, time=0.22, value=1.25}
          {arriveTangent=-1.7745, interpMode=RCIM_Cubic, leaveTangent=-1.7745, tangentMode=RCTM_User, time=0.4, value=1.5}
          {arriveTangent=-0.272, interpMode=RCIM_Cubic, leaveTangent=-0.272, tangentMode=RCTM_User, time=0.8, value=2}
          <at 1.25 0 RCIM_Cubic>
         ]}}}
      Scale Curve=2.5
     )
     Float SG 2:NiagaraFloat=&FloatFromCurve(FloatCurve=DataInterface() {propertyValues {Curve {keys [{tangentMode=RCTM_User, time=0, value=1}, {tangentMode=RCTM_User, time=1, value=3}]}}})
     Float SG 3:NiagaraFloat=0
     Float SG 4:NiagaraFloat=0
     Float SG 5:NiagaraFloat=0
    )
   ) {module=ScaleSpriteSize001}
   &RotateAroundPoint(
    Derive the Rotation Axis=false
    Radius:NiagaraFloat=&FloatFromCurve(FloatCurve=DataInterface() {propertyValues {Curve {keys [
         {interpMode=RCIM_Cubic, tangentMode=RCTM_User, time=0, value=0}
         {interpMode=RCIM_Cubic, tangentMode=RCTM_User, time=0.6, value=60}
         <at 1 0 RCIM_Cubic>
        ]}}})
    Rotation Axis:Vector3f=&RandomRangeVector(Minimum=<-1 -1 -1>, Maximum=<1 1 1>)
    Rotation Center:NiagaraPosition=&AddVectorToPosition(Position=$StackContext.SkeletalMeshLocation.SampledPosition, Vector=<-120 0 0>)
    Rotation Rate:NiagaraFloat=&VC_FloatBySpawnGroup(
     SpawnGroup or ExecIndex:NiagaraInt32=0
     CustomInt:NiagaraInt32=0
     MaximumSpawnGroups:NiagaraInt32=3
     Float SG 0:NiagaraFloat=&FloatFromCurve(FloatCurve=DataInterface() {propertyValues {Curve {keys [{interpMode=RCIM_Cubic, tangentMode=RCTM_User, time=0, value=0}, {interpMode=RCIM_Cubic, tangentMode=RCTM_User, time=0.6, value=0.4}]}}})
     Float SG 1:NiagaraFloat=&FloatFromCurve(FloatCurve=DataInterface() {propertyValues {Curve {keys [{interpMode=RCIM_Cubic, tangentMode=RCTM_User, time=0, value=0}, {interpMode=RCIM_Cubic, tangentMode=RCTM_User, time=0.6, value=0.4}]}}})
     Float SG 2:NiagaraFloat=0
     Float SG 3:NiagaraFloat=0
     Float SG 4:NiagaraFloat=0
     Float SG 5:NiagaraFloat=0
    )
    Set Particle Position:NiagaraBool=&MaskBoolBySpawnGroup(Masked Bool:NiagaraBool=false, Passthrough Bool:NiagaraBool=true, Spawn Group Mask:NiagaraInt32=2)
    X Axis:Vector3f=&RandomRangeVector(Minimum=<-1 -1 -1>, Maximum=<1 1 1>)
   )
   &Drag(
    Drag=&VC_FloatBySpawnGroup(
     SpawnGroup or ExecIndex:NiagaraInt32=0
     CustomInt:NiagaraInt32=0
     MaximumSpawnGroups:NiagaraInt32=3
     Float SG 0:NiagaraFloat=&FloatFromCurve(FloatCurve=DataInterface() {propertyValues {Curve {keys [<at 0.3 0>,<at 0.5 35>]}}}, Scale Curve=10)
     Float SG 1:NiagaraFloat=&FloatFromCurve(FloatCurve=DataInterface() {propertyValues {Curve {keys [<at 0.3 0>,<at 1 35>]}}}, Scale Curve=2)
     Float SG 2:NiagaraFloat=0
     Float SG 3:NiagaraFloat=0
     Float SG 4:NiagaraFloat=0
     Float SG 5:NiagaraFloat=0
    )
    Rotational Drag=0
    Use Linear Drag=true
    Use Rotational Drag=true
   ) {module=Drag001}
   &CurlNoiseForce(
    Noise Strength=&VC_FloatBySpawnGroup(
     SpawnGroup or ExecIndex:NiagaraInt32=0
     CustomInt:NiagaraInt32=0
     MaximumSpawnGroups:NiagaraInt32=3
     Float SG 0:NiagaraFloat=5000
     Float SG 1:NiagaraFloat=5000
     Float SG 2:NiagaraFloat=0
     Float SG 3:NiagaraFloat=0
     Float SG 4:NiagaraFloat=0
     Float SG 5:NiagaraFloat=0
    )
    Pan Noise=false
   ) {module=CurlNoiseForce001}
   &SubUVAnimation(
    "Sprite Renderer "=DataInterface() {propertyValues {SpriteRenderer=&self:BCKP_LocalSprites002_0.NiagaraSpriteRendererProperties_8}}
    UseEndFrame=false
    UseStartFrame=false
   )
   &SolveForcesAndVelocity(Clamp Velocity=false, Limit Acceleration=false)
   /SetVariables(Particles.Fuego2:Vector2f=&Multiply_Vector2D_ByFloat(Float=1.3, Vector2D=$Particles.SpriteSize)) {isSetParameters=true}
  ]
  rendererClasses [&NiagaraSpriteRendererProperties]
  renderers [
   {rendererClass=&NiagaraSpriteRendererProperties}
   {rendererClass=&NiagaraSpriteRendererProperties, rendererIndex=1}
   {rendererClass=&NiagaraSpriteRendererProperties, rendererIndex=2}
   {rendererClass=&NiagaraSpriteRendererProperties, rendererIndex=3}
  ]
 }
 BCKP_Sprites001 {
  simTarget=GPUComputeSim
  emitterUpdateScript [
   &EmitterState()
   &VC_FilteredSocketCount()
   &SpawnRate(SpawnRate=30, Use Spawn Probability=false)
   &SpawnRate(SpawnGroup=1, SpawnRate=80, Use Spawn Probability=false) {module=SpawnRate001}
   &SpawnRate(SpawnGroup=2, SpawnRate=30, Use Spawn Probability=false) {module=SpawnRate002}
   &SpawnRate(SpawnGroup=3, SpawnRate=30, Use Spawn Probability=false) {module=SpawnRate003}
  ]
  particleSpawnScript [
   &VC_PaintedVariantsOverride(
    Black=<rgba 0 0 0 1>
    BurntSienna=<rgba 0.012 0.0037 0.0015 1>
    Cobalt=<rgba 0 0.0003 0.018 1>
    Crimson=<rgba 0.017 0 0.0005 1>
    DefaultColor=<rgba 0.0438 0 0.05 1>
    ForestGreen=<rgba 0 0.019 0.001 1>
    Gold:LinearColor=<rgba 0.1 0.05 0 1>
    Grey=<rgba 0.035 0.035 0.035 1>
    Lime=<rgba 0.02 0.04 0 1>
    Orange=<rgba 0.028 0.0098 0 1>
    Pink=<rgba 0.019 0.0023 0.0129 1>
    Purple=<rgba 0.0103 0 0.025 1>
    R1 Black:LinearColor=<rgba 0 0 0 1>
    R1 BurntSienna:LinearColor=<rgba 0.0642 0.0198 0.0079 1>
    R1 Cobalt:LinearColor=<rgba 0.01 0.0205 0.1 1>
    R1 Crimson:LinearColor=<rgba 0.09 0 0.0024 1>
    R1 DefaultColor=<rgba 0.2 0.0394 0.0073 1>
    R1 ForestGreen:LinearColor=<rgba 0 0.07 0.0035 1>
    R1 Gold:LinearColor=<rgba 0.05 0.0292 0 1>
    R1 Grey:LinearColor=<rgba 0.02 0.02 0.02 1>
    R1 Lime:LinearColor=<rgba 0.075 0.15 0 1>
    R1 Orange:LinearColor=<rgba 0.14 0.0535 0.0004 1>
    R1 Pink:LinearColor=<rgba 0.18 0 0.09 1>
    R1 Purple:LinearColor=<rgba 0.025 0 0.05 1>
    R1 Saffron:LinearColor=<rgba 0.07 0.07 0 1>
    R1 SkyBlue:LinearColor=<rgba 0 0.0333 0.05 1>
    R1 TitaniumWhite:LinearColor=<rgba 0.07 0.07 0.07 1>
    R2 Black:LinearColor=<rgba 0 0 0 1>
    R2 BurntSienna:LinearColor=<rgba 1 0.3333 0 1>
    R2 Cobalt:LinearColor=<rgba 0.19 0.46 1 1>
    R2 DefaultColor:LinearColor=<rgba 2 0.7733 0.16 1>
    R2 Gold:LinearColor=<rgba 1 0.3667 0.05 1>
    R2 Grey:LinearColor=<rgba 0.4 0.4 0.4 1>
    R2 Lime:LinearColor=<rgba 0.3958 1 0.1 1>
    R2 Orange:LinearColor=<rgba 2 0.5 0 1>
    R2 Pink:LinearColor=<rgba 2 0.2396 1.3598 1>
    R2 Purple:LinearColor=<rgba 1.3333 0 2 1>
    R2 Saffron:LinearColor=<rgba 2.5 1.9167 0 1>
    R2 SkyBlue:LinearColor=<rgba 0 1.3308 2 1>
    R3 Black:LinearColor=<rgba 0.0675 0.0675 0.075 1>
    R3 BurntSienna:LinearColor=<rgba 0.3 0.11 0 1>
    R3 Cobalt:LinearColor=<rgba 0.075 0.2247 0.3 1>
    R3 Crimson:LinearColor=<rgba 0.2 0 0.0054 1>
    R3 DefaultColor:LinearColor=<rgba 1 0.2667 0.2 1>
    R3 ForestGreen:LinearColor=<rgba 0 0.5 0.0253 1>
    R3 Gold:LinearColor=<rgba 0.8417 1 0.1 1>
    R3 Grey:LinearColor=<rgba 0.22 0.22 0.22 1>
    R3 Orange:LinearColor=<rgba 2 0.6976 0 1>
    R3 Pink:LinearColor=<rgba 0.3 0.24 0.2782 1>
    R3 Purple:LinearColor=<rgba 0.1642 0 0.4 1>
    R4 DefaultColor:LinearColor=&MakeCustomLinearColorFromBool(Bool=$User.bIsPainted, True Color=$User.OutlineParticleColor, False Color=<rgba 1 0.6939 0.1144 1>)
    Saffron=<rgba 0.02 0.02 0 1>
    SkyBlue=<rgba 0 0.0292 0.035 1>
    Use R1 Painted Overrides=true
    Use R2 Painted Overrides=true
    Use R3 Painted Overrides=true
    Use R4 Painted Overrides=true
   )
   /SetVariables(Particles.VisTag:NiagaraInt32=&SpawnGroup()) {isSetParameters=true}
   &InitializeParticle(
    Color Mode=::NewEnumerator0
    Lifetime=&VC_FloatBySpawnGroup(
     SpawnGroup or ExecIndex:NiagaraInt32=0
     CustomInt:NiagaraInt32=0
     MaximumSpawnGroups:NiagaraInt32=5
     Float SG 0:NiagaraFloat=&RandomRangeFloat(Minimum=0.9, Maximum=1.2)
     Float SG 1:NiagaraFloat=&RandomRangeFloat(Minimum=0.75, Maximum=1.5)
     Float SG 2:NiagaraFloat=&RandomRangeFloat(Minimum=0.8, Maximum=1.5)
     Float SG 3:NiagaraFloat=&RandomRangeFloat(Minimum=0.8, Maximum=1.5)
     Float SG 4:NiagaraFloat=0
     Float SG 5:NiagaraFloat=0
    )
    Sprite Rotation Angle=&RandomRangeFloat(Minimum=-1, Maximum=1)
    Sprite Rotation Mode=::NewEnumerator1
    Sprite Size Mode=::NewEnumerator3
    Uniform Sprite Size=&VC_FloatBySpawnGroup(
     SpawnGroup or ExecIndex:NiagaraInt32=0
     CustomInt:NiagaraInt32=0
     MaximumSpawnGroups:NiagaraInt32=5
     Float SG 0:NiagaraFloat=&RandomRangeFloat(Minimum=10, Maximum=20)
     Float SG 1:NiagaraFloat=30
     Float SG 2:NiagaraFloat=&RandomRangeFloat(Minimum=10, Maximum=20)
     Float SG 3:NiagaraFloat=&RandomRangeFloat(Minimum=12, Maximum=23)
     Float SG 4:NiagaraFloat=0
     Float SG 5:NiagaraFloat=0
    )
    UsePositionOffset=false
    Write Material Random=true
   )
   &PartitionParticles(Number Of Partitions=$Emitter.PortCount)
   &SkeletalMeshLocation(
    Bone / Socket Index=$Particles.PartitionParticles.Partition
    Bone / Socket Sampling Mode=::NewEnumerator5
    MaskSpawnGroup=false
    Mesh Sampling Type=::NewEnumerator6
    Sampled Position Offset=&AddVectorToPosition(
     Position=&VC_VectorBySpawnGroup(
      SpawnGroup or ExecIndex:NiagaraInt32=0
      CustomInt:NiagaraInt32=0
      MaximumSpawnGroups:NiagaraInt32=5
      Vector Group 0:Vector3f=<-130 0 0>
      Vector Group 1:Vector3f=<0 0 0>
      Vector Group 2:Vector3f=<30 0 0>
      Vector Group 3:Vector3f=<30 0 0>
      Vector Group 4:Vector3f=<-50 0 0>
      Vector Group 5:Vector3f=<0 0 0>
     )
     Vector=&AddVectorToPosition(Position=$User.SingleBoostPos, Vector=$User.DTCarOffset)
    )
    Skeletal Mesh=$System.SkeletalMesh
   )
   &AddRotationalVelocity(Rotation Rate=&VectorFromFloat(Value=&VC_FloatBySpawnGroup(
      SpawnGroup or ExecIndex:NiagaraInt32=0
      CustomInt:NiagaraInt32=0
      MaximumSpawnGroups:NiagaraInt32=5
      Float SG 0:NiagaraFloat=&MakeFloatFromVector(
       Vector=&VectorFromFloat(Value=&RandomRangeFloat(Minimum=-0.25, Maximum=0.25))
       Channel=::NewEnumerator0
      )
      Float SG 1:NiagaraFloat=&RandomRangeFloat(Minimum=-0.5, Maximum=0.5)
      Float SG 2:NiagaraFloat=&RandomRangeFloat(Minimum=-0.5, Maximum=0.5)
      Float SG 3:NiagaraFloat=&RandomRangeFloat(Minimum=-0.5, Maximum=0.5)
      Float SG 4:NiagaraFloat=0
      Float SG 5:NiagaraFloat=0
     )))
   &AddVelocity(Velocity=&MakeVector(
     X=&VC_FloatBySpawnGroup(
      SpawnGroup or ExecIndex:NiagaraInt32=0
      CustomInt:NiagaraInt32=0
      MaximumSpawnGroups:NiagaraInt32=5
      Float SG 0:NiagaraFloat=-600
      Float SG 1:NiagaraFloat=-1000
      Float SG 2:NiagaraFloat=-600
      Float SG 3:NiagaraFloat=-600
      Float SG 4:NiagaraFloat=0
      Float SG 5:NiagaraFloat=0
     )
     Y=0
     Z=&VC_FloatBySpawnGroup(
      SpawnGroup or ExecIndex:NiagaraInt32=0
      CustomInt:NiagaraInt32=0
      MaximumSpawnGroups:NiagaraInt32=5
      Float SG 0:NiagaraFloat=0
      Float SG 1:NiagaraFloat=100
      Float SG 2:NiagaraFloat=0
      Float SG 3:NiagaraFloat=0
      Float SG 4:NiagaraFloat=0
      Float SG 5:NiagaraFloat=0
     )
    )) {module=AddVelocity001}
   &AddVelocity(Velocity=&RandomRangeVector(
     Minimum=&MakeVector(
      X=0
      Y=&VC_FloatBySpawnGroup(
       SpawnGroup or ExecIndex:NiagaraInt32=0
       CustomInt:NiagaraInt32=0
       MaximumSpawnGroups:NiagaraInt32=5
       Float SG 0:NiagaraFloat=-50
       Float SG 1:NiagaraFloat=-50
       Float SG 2:NiagaraFloat=-50
       Float SG 3:NiagaraFloat=-50
       Float SG 4:NiagaraFloat=0
       Float SG 5:NiagaraFloat=0
      )
      Z=&VC_FloatBySpawnGroup(
       SpawnGroup or ExecIndex:NiagaraInt32=0
       CustomInt:NiagaraInt32=0
       MaximumSpawnGroups:NiagaraInt32=5
       Float SG 0:NiagaraFloat=-75
       Float SG 1:NiagaraFloat=-100
       Float SG 2:NiagaraFloat=-50
       Float SG 3:NiagaraFloat=-50
       Float SG 4:NiagaraFloat=0
       Float SG 5:NiagaraFloat=0
      )
     )
     Maximum=&MakeVector(
      X=0
      Y=&VC_FloatBySpawnGroup(
       SpawnGroup or ExecIndex:NiagaraInt32=0
       CustomInt:NiagaraInt32=0
       MaximumSpawnGroups:NiagaraInt32=5
       Float SG 0:NiagaraFloat=50
       Float SG 1:NiagaraFloat=50
       Float SG 2:NiagaraFloat=50
       Float SG 3:NiagaraFloat=50
       Float SG 4:NiagaraFloat=0
       Float SG 5:NiagaraFloat=0
      )
      Z=&VC_FloatBySpawnGroup(
       SpawnGroup or ExecIndex:NiagaraInt32=0
       CustomInt:NiagaraInt32=0
       MaximumSpawnGroups:NiagaraInt32=5
       Float SG 0:NiagaraFloat=75
       Float SG 1:NiagaraFloat=50
       Float SG 2:NiagaraFloat=50
       Float SG 3:NiagaraFloat=50
       Float SG 4:NiagaraFloat=0
       Float SG 5:NiagaraFloat=0
      )
     )
    )) {module=AddVelocity002}
   &ShapeLocation(
    MaskSpawnGroup=false
    Override Seed=false
    Sphere Radius=&VC_FloatBySpawnGroup(
     SpawnGroup or ExecIndex:NiagaraInt32=0
     CustomInt:NiagaraInt32=0
     MaximumSpawnGroups:NiagaraInt32=5
     Float SG 0:NiagaraFloat=25
     Float SG 1:NiagaraFloat=25
     Float SG 2:NiagaraFloat=6
     Float SG 3:NiagaraFloat=2
     Float SG 4:NiagaraFloat=0
     Float SG 5:NiagaraFloat=0
    )
    UseOwnerScale=true
   )
   &AddVelocity(Velocity=<-1500 0 0>) {enabled=false}
   /SetVariables(Particles.TesSortOrder=1) {isSetParameters=true}
  ]
  particleUpdateScript [
   &ParticleState()
   &ScaleSpriteSize(
    Scale Sprite Size Mode=::NewEnumerator0
    Uniform Scale Factor=&VC_FloatBySpawnGroup(
     SpawnGroup or ExecIndex:NiagaraInt32=0
     CustomInt:NiagaraInt32=0
     MaximumSpawnGroups:NiagaraInt32=5
     Float SG 0:NiagaraFloat=&FloatFromCurve(
      FloatCurve=DataInterface() {propertyValues {Curve {keys [
          <at 0 0.05 RCIM_Cubic>
          {arriveTangent=1.5079, leaveTangent=1.5079, time=0.1, value=0.95}
          {arriveTangent=-1.0556, leaveTangent=-1.0556, time=0.63, value=1}
          <at 1 0 RCIM_Cubic>
         ]}}}
      Scale Curve=2
     )
     Float SG 1:NiagaraFloat=&FloatFromCurve(
      FloatCurve=DataInterface() {propertyValues {Curve {keys [
          <at 0 0.05 RCIM_Cubic>
          {arriveTangent=4.75, interpMode=RCIM_Cubic, leaveTangent=4.75, time=0.1, value=0.05}
          {arriveTangent=1.7925, interpMode=RCIM_Cubic, leaveTangent=1.7925, time=0.2, value=1}
          {arriveTangent=-1.25, interpMode=RCIM_Cubic, leaveTangent=-1.25, time=0.63, value=1}
          <at 1 0 RCIM_Cubic>
         ]}}}
      Scale Curve=1.5
     )
     Float SG 2:NiagaraFloat=&FloatFromCurve(
      FloatCurve=DataInterface() {propertyValues {Curve {keys [
          <at 0 0.05 RCIM_Cubic>
          {arriveTangent=2.5, interpMode=RCIM_Cubic, leaveTangent=2.5, time=0.1, value=0.05}
          {arriveTangent=1.6667, interpMode=RCIM_Cubic, leaveTangent=1.6667, time=0.4, value=1.05}
          {arriveTangent=-1.75, interpMode=RCIM_Cubic, leaveTangent=-1.75, time=0.7, value=1.05}
          <at 1 0 RCIM_Cubic>
         ]}}}
      Scale Curve=1.5
     )
     Float SG 3:NiagaraFloat=&FloatFromCurve(
      FloatCurve=DataInterface() {propertyValues {Curve {keys [
          <at 0 0.05 RCIM_Cubic>
          {arriveTangent=2.5, interpMode=RCIM_Cubic, leaveTangent=2.5, time=0.1, value=0.05}
          {arriveTangent=1.6667, interpMode=RCIM_Cubic, leaveTangent=1.6667, time=0.4, value=1.05}
          {arriveTangent=-1.75, interpMode=RCIM_Cubic, leaveTangent=-1.75, time=0.7, value=1.05}
          <at 1 0 RCIM_Cubic>
         ]}}}
      Scale Curve=1.5
     )
     Float SG 4:NiagaraFloat=0
     Float SG 5:NiagaraFloat=0
    )
   ) {module=ScaleSpriteSize001}
   &Drag(
    Drag=&VC_FloatBySpawnGroup(
     SpawnGroup or ExecIndex:NiagaraInt32=0
     CustomInt:NiagaraInt32=0
     MaximumSpawnGroups:NiagaraInt32=4
     Float SG 0:NiagaraFloat=&MakeCustomFloatFromBool(
      Bool=$User.bIsFrontendPreview
      True Float=&FloatFromCurve(
       FloatCurve=DataInterface() {propertyValues {Curve {keys [<at 0 0 RCIM_Cubic>, {arriveTangent=0.15, interpMode=RCIM_Cubic, leaveTangent=0.15, time=0.5, value=3}, <at 1 0.15 RCIM_Cubic>]}}}
       Scale Curve=0.25
      )
      False Float=0
     )
     Float SG 1:NiagaraFloat=&FloatFromCurve(
      FloatCurve=DataInterface() {propertyValues {Curve {keys [{tangentMode=RCTM_User, time=0, value=0}, <at 0.5 25>, <at 0.8 25>, {tangentMode=RCTM_User, time=1, value=0.5}]}}}
      Scale Curve=0.15
     )
     Float SG 2:NiagaraFloat=&MakeCustomFloatFromBool(
      Bool=$User.bIsFrontendPreview
      True Float=&FloatFromCurve(
       FloatCurve=DataInterface() {propertyValues {Curve {keys [<at 0 0>, {arriveTangent=2, interpMode=RCIM_Cubic, leaveTangent=2, time=0.25, value=0}, <at 1.25 2.5 RCIM_Cubic>]}}}
       Scale Curve=0.25
      )
      False Float=0
     )
     Float SG 3:NiagaraFloat=&MakeCustomFloatFromBool(
      Bool=$User.bIsFrontendPreview
      True Float=&FloatFromCurve(FloatCurve=DataInterface() {propertyValues {Curve {keys [<at 0 0>, {arriveTangent=4, interpMode=RCIM_Cubic, leaveTangent=4, time=0.25, value=0}, <at 1.25 5 RCIM_Cubic>]}}})
      False Float=0
     )
     Float SG 4:NiagaraFloat=0
     Float SG 5:NiagaraFloat=0
    )
    Use Linear Drag=true
    Use Rotational Drag=true
   )
   &InheritSourceMovement(
    Applied Position Scale=&Multiply_VectorByFloat(
     Vector=<1 1 1>
     Float=&FloatFromCurve(FloatCurve=DataInterface() {propertyValues {Curve {keys [
          {interpMode=RCIM_Cubic, leaveTangent=-0.0051, tangentMode=RCTM_Break, time=0.3, value=1}
          {arriveTangent=-0.247, interpMode=RCIM_Cubic, tangentMode=RCTM_Break, time=1, value=0.9}
         ]}}})
    )
    ApplyForce=false
    ApplyPosition=true
    ApplyPositionFalloff=false
   ) {module=InheritSourceMovement001}
   &SolveForcesAndVelocity(Clamp Velocity=false, Limit Acceleration=false)
  ]
  rendererClasses [&NiagaraSpriteRendererProperties]
  renderers [
   {rendererClass=&NiagaraSpriteRendererProperties}
   {rendererClass=&NiagaraSpriteRendererProperties, rendererIndex=1}
   {rendererClass=&NiagaraSpriteRendererProperties, rendererIndex=2}
   {rendererClass=&NiagaraSpriteRendererProperties, rendererIndex=3}
  ]
 }
}
summary {
 systemName=NS_Boost_Edge
 emitters [
  {emitterName=BCKP_Sprites001, simTarget=GPUComputeSim, rendererClasses [&NiagaraSpriteRendererProperties]}
  {emitterName=BCKP_LocalSprites002, simTarget=GPUComputeSim, rendererClasses [&NiagaraSpriteRendererProperties]}
 ]
 userVariables [
  User.bIsFrontendPreview {type=NiagaraBool}
  User.bIsPainted {type=NiagaraBool}
  User.ColorVariantID {type=NiagaraInt32}
  User.ContrastHalfMoonPower {type=NiagaraFloat, defaultValue=3, description=""}
  User.DTCarOffset {type=Vector3f, description="", defaultValue {struct=Vector3f, value=<0 0 0>}}
  User.EmissiveEdgePainted {type=NiagaraFloat, defaultValue=10, description=""}
  User.FireValuePainted {type=NiagaraFloat, defaultValue=1, description=""}
  User.FlatColorPainted {type=LinearColor, description="", defaultValue {struct=LinearColor, value=<rgba 1 1 1 1>}}
  User.GradientColor {type=LinearColor, description="", defaultValue {struct=LinearColor, value=<rgba 1 1 1 1>}}
  User.HalfMoonEmissivePow {type=NiagaraFloat, defaultValue=1, description=""}
  User.HalfMoonOverride {type=LinearColor, description="", defaultValue {struct=LinearColor, value=<rgba 1 1 1 1>}}
  User.OutlineParticleColor {type=LinearColor, description="", defaultValue {struct=LinearColor, value=<rgba 1 0.6939 0.1144 1>}}
  User.PaintedColor {type=LinearColor, description="", defaultValue {struct=LinearColor, value=<rgba 1 1 1 1>}}
  User.PaintedScalar {type=NiagaraFloat, defaultValue=1, description=""}
  User.PortCount {type=NiagaraInt32, defaultValue=2}
  User.SaturateCenterPainted {type=NiagaraFloat, defaultValue=1, description=""}
  User.ScaleColorDarkDot {type=NiagaraFloat, defaultValue=0.1, description=""}
  User.ScalePaintedGradient {type=NiagaraFloat, defaultValue=4, description=""}
  User.ScalePaintedVariants {type=NiagaraFloat, defaultValue=2, description=""}
  User.SingleBoostPos {type=NiagaraPosition, description="", defaultValue {}}
 ]
}
