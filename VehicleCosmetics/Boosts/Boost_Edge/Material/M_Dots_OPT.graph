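# Masked unlit Niagara sprite material for dot particles in the Edge vehicle boost cosmetic, drawing circle sprites in several styles.
# A dynamic-parameter sub-UV index picks a frame from a 2x2 flipbook texture, whose channels feed named reroutes alongside particle color.
# Scalar parameters switch between thin or thick outline and flat or half-moon gradient looks through If and Lerp nodes.
# Dynamic parameters supply squared colors, gradient contrast and a painted toggle, feeding the shared emissive adjustment function.
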Material /VehicleCosmetics/Boosts/Boost_Edge/Material/M_Dots_OPT
graph {
 properties {
  bAutomaticallySetUsageInEditor=false
  bUsedWithNiagaraSprites=true
  blendMode=BLEND_Masked
  pixelDepthOffsetMode=PDOM_Legacy
  shadingModel=MSM_Unlit
  usedShadingModels=Unlit
 }
 nodes [
  DynamicParameter() {parameterIndex=2, paramNames [R HalfMoon, G HalfMoon, B HalfMoon, EmissimeMult]}
  n1 = Floor(None=ScalarParameter() {parameterName="Outline = 0/ Flat = 1"})
  n2 = DynamicParameter() {paramNames [Sub UV Index, Contrast Half Moon, IsPainted, Param4]}
  n3 = Saturate(None=Floor(None=@n2.IsPainted))
  n4 = DynamicParameter() {parameterIndex=3, paramNames [R FlatParticle, G FlatParticle, B FlatParticle, EmissiveMult]}
  n5 = NamedRerouteUsage()
  n6 = Multiply(A=Multiply(A=@n4.RGB, B=@n4.RGB), B=@n5)
  n7 = NamedRerouteUsage()
  n8 = DynamicParameter() {parameterIndex=2, paramNames [R HalfMoon, G HalfMoon, B HalfMoon, EmissimeMult]}
  n9 = LinearInterpolate(
   A=If(A=@n3, "A > B"=@n6, "A == B"=@n6, "A < B"=@n5) {constB=0.5}
   B=Multiply(A=Power(Base=@n7, Exp=@n2.Contrast Half Moon), B=Multiply(A=@n8.RGB, B=@n8.RGB))
   Alpha=Saturate(None=ScalarParameter() {parameterName="Flat = 0 / Gradient = 1"})
  ) {constAlpha=1}
  n10 = If(A=@n3, "A > B"=@n9, "A == B"=@n9, "A < B"=Multiply(A=@n9, B=NamedRerouteUsage())) {constB=0.5}
  n11 = Saturate(None=Floor(None=ScalarParameter() {parameterName="Outline Thing  = 0/ Thick = 1"}))
  n12 = NamedRerouteUsage()
  n13 = NamedRerouteUsage()
  n14 = DynamicParameter() {desc=Outline Thick Color, parameterIndex=1, paramNames [R, G, B, Emmisive Multiplier]}
  n15 = Add(A=Multiply(A=Subtract(A=@n12, B=@n13), B=NamedRerouteUsage()), B=Multiply(A=@n13, B=Multiply(A=@n14.RGB, B=@n14.RGB)))
  n16 = NamedRerouteUsage()
  n17 = NamedRerouteUsage()
  n18 = Add(A=Multiply(A=Subtract(A=@n16, B=@n17), B=NamedRerouteUsage()), B=Multiply(A=@n17, B=Constant3Vector() {constant=<rgba 0 0 0 1>}))
  MaterialFunctionCall(Emissive=If(A=Saturate(None=@n1), "A > B"=@n10, "A == B"=@n10, "A < B"=If(A=@n11, "A > B"=@n15, "A == B"=@n15, "A < B"=@n18) {constB=0.5}) {constB=0.5}) {materialFunction=&MF_FX_EmissiveAdjustments}
  n19 = ParticleColor()
  Multiply(A=@n19.A, B=NamedRerouteUsage())
  n20 = Clamp(None=DynamicParameter() {paramNames [Sub UV Index, Contrast Half Moon, IsPainted, Param4]}.Sub UV Index) {maxDefault=4, minDefault=1}
  n21 = Constant2Vector() {g=4, r=1}
  n22 = Constant2Vector() {g=1}
  n23 = Constant() {r=2}
  n24 = TextureSample(UVs=MaterialFunctionCall(
    "Animation  Phase (0-1)"=MaterialFunctionCall(Input=@n20, Input Low=@n21.R, Input High=@n21.R, Target Low=@n22.R, Target High=@n22.R) {materialFunction=&RemapValueRange}.Result
    Number of Rows=@n23
    Number of Columns=@n23
   ) {materialFunction=&FlipBook}.UVs) {samplerType=SAMPLERTYPE_LinearColor, texture=/VehicleCosmetics/Boosts/Boost_Edge/Textures/T_VCVFX_Boost_Edge_SpriteCirclesEdge_RGB}
  NamedRerouteDeclaration(None=@n24.R) {name=TEX R, nodeColor=<rgba 1 0.9844 0 1>}
  NamedRerouteDeclaration(None=@n24.G) {name=TEX G, nodeColor=<rgba 0 1 0 1>}
  NamedRerouteDeclaration(None=@n24.B) {name=TEX B, nodeColor=<rgba 0 0 1 1>}
  NamedRerouteDeclaration(None=@n24.A) {name=TEX A, nodeColor=<rgba 0.9896 0.9844 1 1>}
  NamedRerouteDeclaration(None=@n19.RGB) {name=Particle Color, nodeColor=<rgba 1 0.1218 0 1>}
 ]
}
